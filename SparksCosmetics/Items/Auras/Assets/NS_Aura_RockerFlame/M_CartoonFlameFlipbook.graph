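# Translucent unlit Niagara material for the cartoon flames of the RockerFlame aura cosmetic.
# A channel-packed flipbook, driven by time and particle random with optional UV height flicker, produces the flame mask.
# Step-thresholded red-orange gradients form a two-tone look; dynamic-parameter dissolve thresholds, depth fade and particle alpha shape opacity.
# Switchable spline-mesh WPO combines SplineThicken width flicker with a clamped, UV-weighted offset from a Niagara-set position.
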Material /SparksCosmetics/Items/Auras/Assets/NS_Aura_RockerFlame/M_CartoonFlameFlipbook
graph {
 properties {
  bAutomaticallySetUsageInEditor=false
  bUsedWithNiagaraMeshParticles=true
  bUsedWithNiagaraRibbons=true
  bUsedWithNiagaraSprites=true
  blendMode=BLEND_Translucent
  shadingModel=MSM_Unlit
  usedShadingModels=Unlit
 }
 nodes [
  n1 = Constant3Vector() {constant=<rgba 1 0.5 0 1>}
  n2 = Constant3Vector() {constant=<rgba 1 0 0 1>}
  n3 = TextureCoordinate()
  n4 = Power(Base=OneMinus(None=ComponentMask(None=@n3) {g=true}), Exp=ScalarParameter() {defaultValue=0.75, parameterName=VerticalGradient_Pow}) {constExponent=0.75}
  n5 = Step(Y=ScalarParameter() {defaultValue=0.5, parameterName=FlameEdge_Outline}, X=NamedRerouteUsage()) {constY=0.45}
  n6 = LinearInterpolate(Alpha=NamedRerouteUsage()) {constA=-0.2, constB=0.4}
  n7 = LinearInterpolate(
   A=Multiply(A=LinearInterpolate(A=@n1, B=@n2, Alpha=@n4), B=Add(A=@n5) {constB=0.5})
   B=LinearInterpolate(A=Constant3Vector() {constant=<rgba 2 1 0.2 1>}, B=Constant3Vector() {constant=<rgba 5 0 0 1>}, Alpha=Add(A=@n4, B=@n6))
   Alpha=@n5
  )
  n8 = ParticleColor()
  MaterialFunctionCall(Emissive=Multiply(A=Multiply(A=@n7, B=@n8.RGB), B=ScalarParameter() {defaultValue=1, parameterName=Brightness})) {materialFunction=&MF_FX_EmissiveAdjustments}
  n9 = NamedRerouteUsage()
  n10 = DynamicParameter() {defaultValue=<rgba -60 0 1 1>, paramNames [Width,Dissolve,Param3,Param4]}
  n11 = Add(A=ScalarParameter() {defaultValue=0.15, parameterName=Min}, B=@n10.Dissolve)
  n12 = Add(A=ScalarParameter() {defaultValue=0.25, parameterName=Max}, B=@n10.Dissolve)
  n13 = Saturate(None=Divide(A=Subtract(A=@n9, B=@n11), B=Subtract(A=@n12, B=@n11)))
  n14 = DynamicParameter() {defaultValue=<rgba -60 0 1 1>, paramNames [Width,Dissolve,Param3,Param4]}
  n15 = Add(A=ScalarParameter() {parameterName=Min1}, B=@n14.Dissolve)
  n16 = Subtract(A=NamedRerouteUsage(), B=@n15)
  n17 = Subtract(A=Add(A=ScalarParameter() {defaultValue=0.5, parameterName=Max2}, B=@n14.Dissolve), B=@n15)
  n18 = Add(A=@n13, B=Multiply(A=OneMinus(None=@n13), B=Divide(A=@n16, B=@n17)))
  n19 = TextureCoordinate()
  n20 = Multiply(A=OneMinus(None=ComponentMask(None=@n19) {g=true})) {constB=3}
  n21 = StaticSwitchParameter(True=Saturate(None=Power(Base=@n20) {constExponent=5}), False=Constant() {r=1}) {defaultValue=true, group=WPO, parameterName=isSplineMesh, sortPriority=0}
  Multiply(A=DepthFade(Opacity=Multiply(A=@n18, B=@n21)) {fadeDistanceDefault=50}, B=ParticleColor().A)
  n22 = Add(A=Multiply(A=Time(), B=ScalarParameter() {defaultValue=4, parameterName=Speed}), B=Multiply(A=ParticleRandom()) {constB=7.77})
  n23 = Constant() {r=2}
  n24 = Multiply(A=TextureCoordinate()) {constB=-1}
  n25 = Round(None=ParticleRandom())
  n26 = TextureCoordinate()
  n27 = Multiply(A=ComponentMask(None=TextureCoordinate()) {g=true}) {constB=0}
  n28 = Multiply(A=Time(), B=ScalarParameter() {defaultValue=0.1, parameterName=HeightFlicker_Speed}) {constB=0.1}
  n29 = Multiply(A=Add(A=Add(A=@n27, B=@n28), B=ParticleRandom()), B=Constant4Vector() {constant=<rgba 0.2 2 4 16>}.RGBA)
  n30 = Divide(A=DotProduct(A=Cosine(None=@n29), B=Constant4Vector() {constant=<rgba 1 1 1 1>}.RGBA)) {constB=4}
  n31 = Multiply(
   A=LinearInterpolate(
    A=ScalarParameter() {defaultValue=1, parameterName=HeightFlicker_Min}
    B=ScalarParameter() {defaultValue=1.5, parameterName=HeightFlicker_Max}
    Alpha=Saturate(None=@n30)
   )
   B=OneMinus(None=ComponentMask(None=@n26) {g=true})
  )
  n32 = Saturate(None=AppendVector(A=ComponentMask(None=@n26) {r=true}, B=OneMinus(None=@n31)))
  n33 = TextureCoordinate()
  n34 = MaterialFunctionCall!(
   "Animation  Phase (0-1)"=@n22
   Number of Rows=@n23
   Number of Columns=@n23
   Texture=TextureObject() {texture=&/T_FlameFlip16_packed}
   UVs=StaticSwitchParameter(
    True=LinearInterpolate(A=@n24, Alpha=@n25)
    False=StaticSwitchParameter(True=@n32, False=@n33) {defaultValue=true, group=WPO, parameterName=isSplineMesh, sortPriority=0}
   ) {group=WPO, parameterName=FlipUVs?, sortPriority=0}
   MipBias/Level
   "Use Mip Bias (T) Level (F)"=StaticBool()
   "Clamp Anim (see tooltip)"
  ) {materialFunction=/BRCosmetics/Accessories/FORT_Backpacks/Backpack_M_MED_BasketSuburb/Materials/MF_FlipBook_PackedChannels}
  n35 = Floor(None=Fmod(A=@n22, B=Constant() {r=4}))
  n36 = LinearInterpolate(A=@n34.R, B=@n34.R, Alpha=Saturate(None=@n35))
  n37 = Saturate(None=Add(A=@n35) {constB=-1})
  n38 = Add(A=@n35) {constB=-2}
  NamedRerouteDeclaration(None=LinearInterpolate(A=LinearInterpolate(A=@n36, B=@n34.B, Alpha=@n37), B=@n34.A, Alpha=Saturate(None=@n38))) {name=FlameTex, nodeColor=<rgba 1 0.6328 0 1>}
  n39 = Time()
  n40 = ParticleRandom()
  n41 = LinearInterpolate(Alpha=ParticleRandom()) {constA=0.7, constB=1.3}
  n42 = ScalarParameter() {defaultValue=2, parameterName=SineSpeed}
  n43 = Add(A=ComponentMask(None=TextureCoordinate()) {g=true}, B=Multiply(A=Add(A=@n39, B=@n40), B=Multiply(A=@n41, B=@n42) {constB=2}))
  n44 = ScalarParameter() {defaultValue=1, group=WPO, parameterName=SineScale, sortPriority=10}
  NamedRerouteDeclaration(None=Sine(None=Multiply(A=@n43, B=@n44))) {name=PanningSine, nodeColor=<rgba 0 1 0.0625 1>}
  n45 = ScalarParameter() {parameterName=WPOThinning_Min}
  n46 = ScalarParameter() {defaultValue=10, parameterName=WPOThinning_Max}
  n47 = ParticleRandom()
  n48 = Multiply(A=LinearInterpolate(Alpha=@n47) {constA=0.35, constB=0.65}, B=Time()) {constB=0.5}
  n49 = ConstantBiasScale(None=Sine(None=Add(A=@n47, B=@n48)))
  n50 = Add(
   A=DynamicParameter() {defaultValue=<rgba -60 0 1 1>, paramNames [Width,Dissolve,Param3,Param4]}.Width
   B=Multiply(A=Constant3Vector() {constant=<rgba 1 1 0 0>}, B=LinearInterpolate(A=@n45, B=@n46, Alpha=@n49))
  )
  n51 = MaterialFunctionCall(WidthBase=@n50, WidthTip=@n50) {materialFunction=&SplineThicken}
  n52 = DynamicParameter() {desc=Parameters set by Niagara, parameterIndex=3, defaultValue=<rgba 0 1 1 1>, paramNames [X,Y,Z,Intensity]}
  n53 = MaterialFunctionCall(Float3=Subtract(
    A=MaterialFunctionCall(X=@n52.X, Y=@n52.X, Z=@n52.X) {materialFunction=&MakeFloat3}.Result
    B=ParticlePositionWS()
   )) {materialFunction=&BreakOutFloat3Components}
  n54 = ScalarParameter() {defaultValue=10, desc=Interpolate Clamp XY, group=WPO, parameterName=Interpolate XY Clamp, sortPriority=3}
  n55 = Clamp(None=AppendVector(A=@n53.R, B=@n53.R), Min=Multiply(A=@n54) {constB=-1}, Max=@n54) {maxDefault=1000, minDefault=-1000}
  n56 = ScalarParameter() {defaultValue=2, desc=Interpolate Clamp Z, group=WPO, parameterName=Interpolate Clamp Z, sortPriority=4}
  n57 = Clamp(None=@n53.B, Min=Multiply(A=@n56) {constB=-1}, Max=@n56) {minDefault=-1}
  n58 = TextureCoordinate()
  n59 = Power(
   Base=ComponentMask(None=OneMinus(None=@n58)) {g=true}
   Exp=ScalarParameter() {defaultValue=5, desc=Interpolate Power, group=WPO, parameterName=Interpolate Power, sortPriority=2}
  )
  n60 = StaticSwitchParameter(True=LinearInterpolate(B=AppendVector(A=@n55, B=@n57), Alpha=Multiply(A=@n52.Intensity, B=@n59)), False=Constant()) {defaultValue=true, group=WPO, parameterName=Use Interpolate WPO, sortPriority=1}
  NamedRerouteDeclaration(None=StaticSwitchParameter(True=Add(A=@n51.WorldPositionOffset, B=@n60), False=Constant()) {defaultValue=true, group=WPO, parameterName=isSplineMesh, sortPriority=0}) {name=WPO, nodeColor=<rgba 0 0 0 1>}
  NamedRerouteUsage()
  ScalarParameter() {parameterName=Dissolve}
  ScalarParameter() {parameterName=Dissolve}
 ]
}
